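; Looping gameplay cue for the GuineaPig weapon's charge start that feeds the weapon's charge level into the cue's looping sound.
; On loop start it caches the target FortPawn and first audio component, updates once, then starts a repeating short-interval timer.
; Each tick writes the current weapon's charge percent to the "ChargePercent" audio float parameter.
; Removal clears the timer; graph notes mention a threshold against stray third-person sound, but no threshold check appears here.
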(event EventOnLoopingStart(EitherVFX-preferthisversion)
  (params (MyTarget :type Actor%20Object%20Reference) (Parameters :type Gameplay%20Cue%20Parameters%20Structure%20%28by%20ref%29) (ParticleComponents :type Array%20of%20FXSystem%20Component%20Object%20References) (AudioComponents :type Array%20of%20Audio%20Component%20Object%20References))
  (bind _asfort_pawn (Utilities|Casting|CastToFortPawn MyTarget))
  (Variables|Default|SetFort%20Pawn _asfort_pawn)
  (Variables|Default|SetLooping%20Audio%20Component (Utilities|Array|Get(acopy) AudioComponents))
  ;[0.106 0.2188 0.2045] Added a small threshold for charge percent when playing the charge sound 3P. This keeps the charge sound from playing 3P when the weapon is not being charged up.
  (CallFunction|SetChargePercent)
  (bind _returnvalue (Utilities|Time|SetTimerbyEvent (delegate Custom|Set%20Charge%20Percent) 0.1 true))
  (Variables|Default|SetCharge%20Percent%20Timer _returnvalue))

(event Custom|Set%20Charge%20Percent
  ;[0.106 0.2188 0.2045] Added a small threshold for charge percent when playing the charge sound 3P. This keeps the charge sound from playing 3P when the weapon is not being charged up.
  (bind _fort_pawn (Variables|Default|GetFort%20Pawn)
    (:then
      (bind _looping_audio_component (Variables|Default|GetLooping%20Audio%20Component)
        (:then
          (Audio|Parameter|SetFloatParameter _looping_audio_component "ChargePercent" (FortWeapon|GetChargePercent (Weapon|GetCurrentWeapon _fort_pawn)) :!class /Script/AudioExtensions.AudioParameterControllerInterface))
        (:else)))
    (:else)))

(event EventOnRemoval(EitherVFX-preferthisversion)
  (params (MyTarget :type Actor%20Object%20Reference) (Parameters :type Gameplay%20Cue%20Parameters%20Structure%20%28by%20ref%29) (ParticleComponents :type Array%20of%20FXSystem%20Component%20Object%20References) (AudioComponents :type Array%20of%20Audio%20Component%20Object%20References) (BurstCameraShakeInstance :type Camera%20Shake%20Base%20Object%20Reference) (BurstDecalInstance :type Decal%20Actor%20Object%20Reference))
  (Utilities|Time|ClearandInvalidateTimerbyHandle MyTarget (Variables|Default|GetCharge%20Percent%20Timer)))

(fn ConstructionScript)
